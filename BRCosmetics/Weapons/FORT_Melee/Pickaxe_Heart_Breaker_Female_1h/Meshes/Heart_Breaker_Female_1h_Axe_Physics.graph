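# Physics asset for the one-handed Heart Breaker female pickaxe, setting up physical bodies for its dangling chain bones.
# Box bodies run from a small base through a parent and three links to a larger sphere at the tip.
# Five constraints join neighboring bones in series with limited swing and very tight twist, with the widest single-axis swing at the base joint so the chain sways without spinning.
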PhysicsAsset /BRCosmetics/Weapons/FORT_Melee/Pickaxe_Heart_Breaker_Female_1h/Meshes/Heart_Breaker_Female_1h_Axe_Physics
bodies {
 dyn_chain_01 [{shapeType=Box, extentX=3.0171, extentY=4.75, extentZ=1.8104, center=<1.4377 0 0>}]
 dyn_chain_02 [{shapeType=Box, extentX=3.0171, extentY=4.75, extentZ=1.8104, center=<1.6037 0 0>}]
 dyn_chain_03 [{shapeType=Box, extentX=2.7671, extentY=4.75, extentZ=1.8104, center=<1.357 0 0>}]
 dyn_chain_04 [{shapeType=Sphere, radius=6.25, center=<5.6334 0 -0.0001>}]
 dyn_chain_base [{shapeType=Box, extentX=1, extentY=1, extentZ=1}]
 dyn_chain_parent [{shapeType=Box, extentX=4.2671, extentY=4.75, extentZ=1.8104, center=<2.1374 0 0>}]
}
constraints [
 {bone1Name=dyn_chain_parent, bone2Name=dyn_chain_base, swing1LimitDegrees=20, swing2LimitDegrees=80, twistLimitDegrees=1}
 {bone1Name=dyn_chain_01, bone2Name=dyn_chain_parent, swing1LimitDegrees=20, swing2LimitDegrees=30, twistLimitDegrees=2}
 {bone1Name=dyn_chain_02, bone2Name=dyn_chain_01, swing1LimitDegrees=25, swing2LimitDegrees=25, twistLimitDegrees=2}
 {bone1Name=dyn_chain_03, bone2Name=dyn_chain_02, swing1LimitDegrees=25, swing2LimitDegrees=20, twistLimitDegrees=2}
 {bone1Name=dyn_chain_04, bone2Name=dyn_chain_03, swing1LimitDegrees=25, swing2LimitDegrees=20, twistLimitDegrees=2}
]
